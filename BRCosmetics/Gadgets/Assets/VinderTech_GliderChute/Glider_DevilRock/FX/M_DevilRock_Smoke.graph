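# Translucent, unlit, two-sided smoke trail material for the DevilRock glider's effects.
# Color blends dark and main tones through a curve atlas row, with optional gradient and particle color.
# Opacity combines a panning trail texture and noises, shaped by curve and edge masks, then depth and camera fades.
# World position offset adds vertex-red-scaled sine/cosine wobble, SplineThicken widening and optional tilt.
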Material /BRCosmetics/Gadgets/Assets/VinderTech_GliderChute/Glider_DevilRock/FX/M_DevilRock_Smoke
graph {
 properties {
  bUsedWithBeamTrails=true
  bUsedWithMeshDeformer=true
  bUsedWithNiagaraRibbons=true
  bUsedWithSkeletalMesh=true
  blendMode=BLEND_Translucent
  pixelDepthOffsetMode=PDOM_Legacy
  shadingModel=MSM_Unlit
  twoSided=true
  usedShadingModels=Unlit
 }
 nodes [
  n1 = ParticleColor()
  n2 = VectorParameter() {parameterName=Use gradient color 01, defaultValue=<rgba 1 1 1 0>}
  n3 = VectorParameter() {parameterName=Use gradient color 02, defaultValue=<rgba 0.0224 0 1 0>}
  n4 = ComponentMask(None=TextureCoordinate()) {g=true}
  n5 = Power(Base=Add(A=Multiply(A=@n4) {constB=0.5}, B=ScalarParameter() {defaultValue=0.2, parameterName=Color lerp offset}) {constB=0.2}) {constExponent=1}
  n6 = TextureCoordinate()
  n7 = ScalarParameter() {parameterName=UV_rotation}
  n8 = CurveAtlasRowParameter(CurveTime=ComponentMask(None=MaterialFunctionCall(UVs=@n6, "Rotation Angle (0-1)"=@n7) {materialFunction=&CustomRotator}.Rotated Values) {g=true}) {
   atlas=&AthenaCurveAtlas01
   curve=/Game/Gadgets/Assets/VinderTech_GliderChute/Glider_DevilRock/FX/Curve_DevilRock
   defaultValue=126
  }
  n9 = LinearInterpolate(
   A=VectorParameter() {parameterName=Dark color, defaultValue=<rgba 0 0 0 1>}.RGB
   B=VectorParameter() {parameterName=Main color, defaultValue=<rgba 1 0.204 0 1>}.RGB
   Alpha=@n8.A
  )
  n10 = StaticSwitchParameter(True=Multiply(A=LinearInterpolate(A=@n2.RGB, B=@n3.RGB, Alpha=@n5), B=@n9), False=@n9) {parameterName=Use gradient color}
  n11 = Multiply(A=@n1.RGB, B=@n10)
  n12 = ScalarParameter() {parameterName=Desaturation}
  MaterialFunctionCall(Emissive=StaticSwitchParameter(True=Desaturation(None=@n11, Fraction=@n12), False=@n10) {parameterName=Use particle color}) {materialFunction=&MF_FX_EmissiveAdjustments}
  n13 = MaterialFunctionCall(UVs=TextureCoordinate(), "Rotation Angle (0-1)"=ScalarParameter() {parameterName=UV_rotation}) {materialFunction=&CustomRotator}
  n14 = ComponentMask(None=@n13.Rotated Values) {g=true}
  n15 = LinearInterpolate(
   A=ScalarParameter() {parameterName=Basic_tweak_Floor}
   B=ScalarParameter() {defaultValue=1, parameterName=Basic_tweak_Ceil}
   Alpha=VertexColor().R
  )
  n16 = AppendVector(A=Constant(), B=Multiply(A=ScalarParameter() {defaultValue=0.4, parameterName=Motion_speed}, B=@n15))
  n17 = ComponentMask(None=Add(A=Panner(Coordinate=@n13.Rotated Values, Speed=@n16) {speedY=1}, B=@n15)) {g=true}
  n18 = AppendVector(A=AppendVector(A=Sine(None=@n17), B=Cosine(None=@n17)), B=Constant())
  n19 = Multiply(A=@n15, B=ScalarParameter() {defaultValue=30, parameterName=Motion_amplitude})
  n20 = Transform(None=Multiply(A=OneMinus(None=@n14), B=Multiply(A=@n18, B=@n19))) {transformSourceType=TRANSFORMSOURCE_Local}
  n21 = ScalarParameter() {defaultValue=101, parameterName=Base_Thickness}
  n22 = ScalarParameter() {defaultValue=60, parameterName=Top_Thickness}
  n23 = MaterialFunctionCall(Float2=TextureCoordinate()) {materialFunction=&BreakOutFloat2Components}
  n24 = ScalarParameter() {defaultValue=0.3, parameterName=UV_Base_thick}
  n25 = ScalarParameter() {defaultValue=1, parameterName=UV_Top_thick}
  n26 = Sine(None=Multiply(A=OneMinus(None=@n23.G)) {constB=0.25})
  n27 = Multiply(A=AppendVector(A=@n23.R, B=LinearInterpolate(A=@n24, B=@n25, Alpha=@n26) {constA=0.3})) {constB=100}
  n28 = Constant()
  n29 = StaticSwitchParameter(
   True=MaterialFunctionCall(
    WidthBase=Multiply(A=@n21, B=@n15)
    WidthTip=Multiply(A=@n15, B=@n22)
    UVs for Thickness=StaticSwitchParameter(True=@n27, False=@n28) {parameterName=UV_Thickness}
   ) {materialFunction=&SplineThicken}.WorldPositionOffset
   False=Constant()
  ) {defaultValue=true, parameterName=Use_SplineThicken}
  n30 = TextureCoordinate()
  n31 = Constant3Vector() {constant=<rgba 0 -1 0 0>}
  n32 = ScalarParameter() {defaultValue=-150, parameterName=WPO_TiltScale}
  n33 = Divide(A=@n32) {constB=8}
  n34 = VertexColor()
  n35 = Multiply(
   A=OneMinus(None=ComponentMask(None=@n30) {g=true})
   B=Multiply(A=Transform(None=@n31) {transformSourceType=TRANSFORMSOURCE_Local}, B=LinearInterpolate(A=@n33, B=@n32, Alpha=@n34.R))
  )
  n36 = Constant()
  StaticSwitchParameter(True=Add(A=Add(A=@n20, B=@n29), B=StaticSwitchParameter(True=@n35, False=@n36) {parameterName=UseWPO_Tilt}), False=@n36) {defaultValue=true, parameterName=WPO}
  n37 = MaterialFunctionCall(UVs=TextureCoordinate(), "Rotation Angle (0-1)"=Add(A=Constant() {r=0.25}, B=ScalarParameter() {parameterName=UV_rotation})) {materialFunction=&CustomRotator}
  n38 = VectorParameter() {parameterName=Center_tile, defaultValue=<rgba 0.3 1 -0.2 0>}
  n39 = AppendVector(A=@n38.B, B=@n38.B)
  n40 = StaticSwitchParameter(True=DynamicParameter() {paramNames [UV_offset,Param2,Param3,Param4]}.UV_offset, False=VertexColor().R) {parameterName=Particle_offset}
  n41 = Add(
   A=Multiply(A=Panner(Coordinate=@n37.Rotated Values, Speed=@n39), B=ComponentMask(None=@n38.RGB) {g=true, r=true})
   B=AppendVector(A=@n40, B=Constant())
  ) {constB=0}
  n42 = ScalarParameter() {defaultValue=3, parameterName=Enhance}
  n43 = TextureCoordinate()
  n44 = ScalarParameter() {parameterName=UV_rotation}
  n45 = VectorParameter() {parameterName=Noise1_tile, defaultValue=<rgba 0.7 0.2 0 0.5>}
  n46 = Multiply(
   A=Panner(
    Coordinate=MaterialFunctionCall(UVs=@n43, "Rotation Angle (0-1)"=@n44) {materialFunction=&CustomRotator}.Rotated Values
    Speed=AppendVector(A=@n45.B, B=@n45.B)
   )
   B=ComponentMask(None=@n45.RGB) {g=true, r=true}
  )
  n47 = TextureCoordinate()
  n48 = ScalarParameter() {parameterName=UV_rotation}
  n49 = VectorParameter() {parameterName=Noise2_tile, defaultValue=<rgba 1 0.5 0 0.3>}
  n50 = Multiply(
   A=Panner(
    Coordinate=MaterialFunctionCall(UVs=@n47, "Rotation Angle (0-1)"=@n48) {materialFunction=&CustomRotator}.Rotated Values
    Speed=AppendVector(A=@n49.B, B=@n49.B)
   )
   B=ComponentMask(None=@n49.RGB) {g=true, r=true}
  )
  n51 = Multiply(
   A=TextureSample(UVs=Add(A=@n46, B=@n40)) {texture=/Game/Weapons/WeaponSkins/Textures/T_Swirly_Noise_D1}.R
   B=TextureSample(UVs=Add(A=@n40, B=@n50)) {texture=&T_Inky_Smoke_Tile}.R
  )
  n52 = MaterialFunctionCall(UVs=TextureCoordinate(), "Rotation Angle (0-1)"=ScalarParameter() {parameterName=UV_rotation}) {materialFunction=&CustomRotator}
  n53 = MaterialFunctionCall(Float2=@n52.Rotated Values) {materialFunction=&BreakOutFloat2Components}
  n54 = ScalarParameter() {defaultValue=0.5, parameterName=Stretch}
  n55 = SphereMask(A=AppendVector(A=@n53.R, B=LinearInterpolate(B=@n54, Alpha=@n53.G)), B=AppendVector(A=Constant() {r=0.5}, B=@n54)) {attenuationRadius=0.5, hardnessPercent=0}
  n56 = ComponentMask(None=@n52.Rotated Values) {g=true}
  n57 = CurveAtlasRowParameter(CurveTime=OneMinus(None=@n56)) {
   atlas=&AthenaCurveAtlas01
   curve=&Curve_DevilRock
   defaultValue=126
  }
  n58 = VectorParameter() {parameterName=Noise3_tile, defaultValue=<rgba 2 0.3 0 0.3>}
  n59 = AppendVector(A=@n58.B, B=@n58.B)
  n60 = TextureSample(UVs=Multiply(A=Panner(Coordinate=@n52.Rotated Values, Speed=@n59) {speedY=0.2}, B=ComponentMask(None=@n58.RGB) {g=true, r=true})) {texture=&T_Swirly_Noise_D1}
  n61 = Constant() {r=1}
  n62 = OneMinus(None=Power(Base=@n56) {constExponent=2.5})
  n63 = Multiply(
   A=@n57.B
   B=MaterialFunctionCall(In=LinearInterpolate(A=@n60.R, B=@n61, Alpha=@n62), Contrast=Constant() {r=1}) {materialFunction=&CheapContrast}.Result
  )
  n64 = VectorParameter() {group=Trailing Edge Sphere Mask, parameterName=Trailing Edge Sphereical UV Scale, sortPriority=4, defaultValue=<rgba 1 1 0 0>}
  n65 = Multiply(A=TextureCoordinate(), B=ComponentMask(None=@n64.RGB) {g=true, r=true})
  n66 = VectorParameter() {group=Trailing Edge Sphere Mask, parameterName="Trailing Edge Sphere Mask Coord (B)", sortPriority=3, defaultValue=<rgba 1.5 1.8 0 0>}
  n67 = ComponentMask(None=@n66.RGB) {g=true, r=true}
  n68 = ScalarParameter() {defaultValue=2.2, group=Trailing Edge Sphere Mask, parameterName=Trailing Edge Radius, sortPriority=1}
  n69 = ScalarParameter() {defaultValue=0.7, group=Trailing Edge Sphere Mask, parameterName=Trailing Edge Hardness, sortPriority=2}
  n70 = Power(
   Base=OneMinus(None=SphereMask(A=@n65, B=@n67, Radius=@n68, Hardness=@n69))
   Exp=ScalarParameter() {defaultValue=5, group=Trailing Edge Sphere Mask, parameterName=Trailing Edge Contrast}
  ) {constExponent=5}
  n71 = ScalarParameter() {defaultValue=1, group=Alpha, parameterName=Multiply Trailing Edge Sphere Mask, sortPriority=6}
  n72 = ComponentMask(None=TextureCoordinate()) {g=true}
  n73 = Add(A=@n72, B=ScalarParameter() {group="Fade Top && Bottom Length Edges", parameterName=Top Length Fade Bias, sortPriority=2})
  n74 = ScalarParameter() {defaultValue=5, group="Fade Top && Bottom Length Edges", parameterName=Top Length Fade Scale}
  n75 = LinearInterpolate(
   B=Saturate(None=Multiply(A=@n73, B=@n74) {constB=3})
   Alpha=ScalarParameter() {defaultValue=1, group=Alpha, parameterName=Fade Top Length Edge, sortPriority=7}
  ) {constA=1}
  n76 = Add(A=OneMinus(None=@n72), B=ScalarParameter() {group="Fade Top && Bottom Length Edges", parameterName=Bottom Length Fade Bias, sortPriority=3})
  n77 = ScalarParameter() {defaultValue=5, group="Fade Top && Bottom Length Edges", parameterName=Bottom Length Fade Scale, sortPriority=1}
  n78 = LinearInterpolate(
   B=Saturate(None=Multiply(A=@n76, B=@n77) {constB=2})
   Alpha=ScalarParameter() {defaultValue=1, group=Alpha, parameterName=Fade Bottom Length Edge, sortPriority=8}
  ) {constA=1}
  n79 = ComponentMask(None=TextureCoordinate()) {r=true}
  n80 = ScalarParameter() {group="Fade Leading && Trailing Edges", parameterName=Leading Edge Fade Bias, sortPriority=2}
  n81 = Saturate(None=Multiply(A=Add(A=@n79, B=@n80), B=ScalarParameter() {defaultValue=3, group="Fade Leading && Trailing Edges", parameterName=Leading Edge Fade Scale}) {constB=3})
  n82 = ScalarParameter() {defaultValue=1, group=Alpha, parameterName=Fade Leading Edge, sortPriority=4}
  n83 = OneMinus(None=@n79)
  n84 = ScalarParameter() {group="Fade Leading && Trailing Edges", parameterName=Trailing Edge Fade Bias, sortPriority=3}
  n85 = Saturate(None=Multiply(
    A=Add(A=@n83, B=@n84)
    B=ScalarParameter() {defaultValue=2, group="Fade Leading && Trailing Edges", parameterName=Trailing Edge Fade Scale, sortPriority=1}
   ) {constB=2})
  n86 = ScalarParameter() {defaultValue=1, group=Alpha, parameterName=Fade Trailing Edge, sortPriority=5}
  n87 = Multiply(
   A=Multiply(A=Multiply(A=@n75, B=@n78), B=LinearInterpolate(B=@n70, Alpha=@n71) {constA=1})
   B=Multiply(A=LinearInterpolate(B=@n81, Alpha=@n82) {constA=1}, B=LinearInterpolate(B=@n85, Alpha=@n86) {constA=1})
  )
  n88 = Multiply(
   A=Add(
    A=TextureSample(UVs=@n41) {samplerType=SAMPLERTYPE_Grayscale, texture=&T_Smoke_Trail_Soft}.R
    B=LinearInterpolate(B=@n42, Alpha=@n51)
   )
   B=LinearInterpolate(A=Multiply(A=@n55, B=@n63), B=Multiply(A=@n63, B=@n87), Alpha=ScalarParameter() {parameterName=UseCurvedAlpha, sliderMax=1})
  )
  n89 = DepthFade(Opacity=@n88, FadeDistance=ScalarParameter() {defaultValue=3, parameterName=FadeDistance})
  n90 = Multiply(
   A=ShadingPathSwitch(Default=@n89, Deferred=@n89, Mobile=@n88) {desc=Migrated from FeatureLevelSwitch}
   B=MaterialFunctionCall(Distance=ScalarParameter() {parameterName=CameraFade}) {materialFunction=&Near_Camera_Fade}.Result
  )
  StaticSwitchParameter(True=Multiply(A=@n1.A, B=@n90), False=Multiply(A=@n90, B=ScalarParameter() {defaultValue=1, parameterName=OpacityScale})) {parameterName=Use particle color}
 ]
}
