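# Niagara deploy effect for the DevilRock glider, combining chest fire, a fire trail, wing sparks and wing ribbons.
# Sprite emitters sample a shared skeletal mesh reader: delayed fire puffs spawn by rate and by distance with speed-dependent probability, while hot orange Snow_wings sparks alternate between bones.
# Wings_Trail builds two anim-trail ribbons tinted by time of day.
# Particles inherit scaled owner movement, and front-end modifiers adjust offsets, spark velocity and trail alpha.
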NiagaraSystem /BRCosmetics/Gadgets/Assets/VinderTech_GliderChute/Glider_DevilRock/FX/NS_Glider_DevilRock_Deploy
dependencies {
 usedDataInterfaces [&NiagaraDataInterfaceCurve, &NiagaraDataInterfaceColorCurve, &NiagaraDataInterfaceSkeletalMesh]
 usedDynamicInputs [&Vector2DFromFloat]
 usedModules [
  &self:SystemSpawnScript.SystemScriptSource.SystemScriptGraph.NiagaraNodeAssignment_0.SetVariables_#1
  &self:SystemSpawnScript.SystemScriptSource.SystemScriptGraph.NiagaraNodeAssignment_1.SetVariables_#2
  &SystemState
 ]
}
emitters {
 Fire_Chest {
  emitterUpdateScript [
   &EmitterState(Life Cycle Mode=::NewEnumerator1, Loop Behavior=::NewEnumerator1, Loop Delay=0.4, Loop Duration=0.5, UseLoopDelay=true)
   &SpawnRate(SpawnRate=15, Use Spawn Probability=false)
  ]
  particleSpawnScript [
   &InitializeParticle(
    Color=&RandomRangeLinearColor(Minimum=<rgba 2 1.2022 0.6482 0.4>, Maximum=<rgba 2 1.2022 0.6482 1>)
    Lifetime=&RandomRangeFloat(Minimum=0.6, Maximum=0.8)
    Position Offset=<0 100 0>
    Sprite Rotation Mode=::NewEnumerator2
    Sprite Size Mode=::NewEnumerator1
    Uniform Sprite Size Max=180
    Uniform Sprite Size Min=130
    UsePositionOffset=true
    Write Material Random=true
   )
   &ShapeLocation(MaskSpawnGroup=false, Override Seed=false, Sphere Radius=70, UseOwnerScale=true)
   &CameraOffset(Camera Offset Amount=&FN_FrontEndFloatModifier(Gameplay=50, Front End=20))
  ]
  particleUpdateScript [
   &ParticleState()
   &SkeletalMeshLocation(Bone Sampling Mode=::NewEnumerator4, MaskSpawnGroup=false, Skeletal Mesh=$System.Skeletal Mesh Reader)
   &SpriteRotationRate(Rotation Rate=1.7329)
   &InheritSourceMovement(Applied Position Scale=<0.5 0.5 0.05>, ApplyForce=false, ApplyPosition=true, ApplyPositionFalloff=false)
   &DynamicMaterialParameters(
    Index 0 Param 1=&FloatFromCurve(FloatCurve=DataInterface() {propertyValues {Curve {keys [<at 0 0>,<at 1 -1>]}}})
    Index 0 Param 2=1
    Index 0 Param 3=&FloatFromCurve(FloatCurve=DataInterface() {propertyValues {Curve {defaultValue=1, keys [{tangentMode=RCTM_User, time=0.25, value=1}, {tangentMode=RCTM_User, time=0.8, value=0}]}}})
    Index 0 Param 4=0.5
    Module.Index0Float0Write=true
    Module.Index0Float1Write=true
    Module.Index0Float2Write=true
    Module.Index0Float3Write=true
   )
   &ScaleSpriteSize(Uniform Curve Sprite Scale=DataInterface() {propertyValues {Curve {
       defaultValue=0.3
       keys [{interpMode=RCIM_Cubic, tangentMode=RCTM_User, time=0, value=0}, {interpMode=RCIM_Cubic, tangentMode=RCTM_User, time=1, value=1}]
      }}})
   &SolveForcesAndVelocity(Clamp Velocity=false, Limit Acceleration=false)
   &ScaleColor(
    Scale Alpha=&FloatFromCurve(FloatCurve=DataInterface() {propertyValues {Curve {
        defaultValue=0
        keys [
         {interpMode=RCIM_Cubic, tangentMode=RCTM_User, time=0, value=0}
         {interpMode=RCIM_Cubic, tangentMode=RCTM_User, time=0.5, value=1}
         {tangentMode=RCTM_User, time=1, value=0}
        ]
       }}})
    Scale RGB=&MakeVectorFromLinearColorRGB(Color=&ColorFromCurve(ColorCurve=DataInterface() {propertyValues {
        AlphaCurve {keys [<at 0 0.6 RCIM_Cubic>,<at 1 0 RCIM_Cubic>]}
        BlueCurve {keys [<at 0 0.0089>,<at 0.3617 1 RCIM_Cubic>,<at 0.8813 0.0089 RCIM_Cubic>]}
        GreenCurve {keys [<at 0 0.0196>,<at 0.3617 1 RCIM_Cubic>,<at 0.8813 0.0196 RCIM_Cubic>]}
        RedCurve {keys [<at 0 0.0521>,<at 0.3617 1 RCIM_Cubic>,<at 0.8813 0.0521 RCIM_Cubic>]}
       }}))
    ScaleA=true
    ScaleRGB=true
   )
  ]
  rendererClasses [&NiagaraSpriteRendererProperties]
  renderers [{rendererClass=&NiagaraSpriteRendererProperties}]
 }
 Fire_DeployTrail {
  emitterUpdateScript [
   &EmitterState(Life Cycle Mode=::NewEnumerator1, Loop Behavior=::NewEnumerator1, Loop Delay=0.4, Loop Duration=0.5, UseLoopDelay=true)
   &SpawnPerUnit(
    Module.Use Max Movement Threshold=true
    Module.Use Movement Tolerance=true
    Movement Tolerance=0.1
    Spawn Probability:NiagaraFloat=&MakeCustomFloatFromBool(
     Bool=&SetBoolByFloatComparison(A=2000, B=&VectorLength(Vector=$Engine.Owner.Velocity))
     True Float=1
     False Float=0.5
    )
    Use Spawn Probability=true
   ) {module=SpawnPerUnit001}
  ]
  particleSpawnScript [
   &InitializeParticle(
    Color=&RandomRangeLinearColor(Minimum=<rgba 2 1.2022 0.6482 0.4>, Maximum=<rgba 2 1.2022 0.6482 1>)
    Lifetime=&RandomRangeFloat(Minimum=0.4, Maximum=0.6)
    Position Offset=<0 100 0>
    Sprite Rotation Mode=::NewEnumerator2
    Sprite Size Mode=::NewEnumerator1
    Uniform Sprite Size Max=250
    Uniform Sprite Size Min=200
    UsePositionOffset=true
    Write Material Random=true
   )
   &ShapeLocation(MaskSpawnGroup=false, Override Seed=false, Sphere Radius=70, UseOwnerScale=true)
   &SkeletalMeshLocation(Bone Sampling Mode=::NewEnumerator4, MaskSpawnGroup=false, Skeletal Mesh=$System.Skeletal Mesh Reader)
   &CameraOffset(Camera Offset Amount=&FN_FrontEndFloatModifier(Gameplay=50, Front End=20))
  ]
  particleUpdateScript [
   &ParticleState()
   &SpriteRotationRate(Rotation Rate=1.7329)
   &DynamicMaterialParameters(
    Index 0 Param 1=&FloatFromCurve(FloatCurve=DataInterface() {propertyValues {Curve {keys [<at 0 0>,<at 1 -1>]}}})
    Index 0 Param 2=1
    Index 0 Param 3=&FloatFromCurve(FloatCurve=DataInterface() {propertyValues {Curve {defaultValue=1, keys [{tangentMode=RCTM_User, time=0.25, value=1}, {tangentMode=RCTM_User, time=0.8, value=0}]}}})
    Index 0 Param 4=0.5
    Module.Index0Float0Write=true
    Module.Index0Float1Write=true
    Module.Index0Float2Write=true
    Module.Index0Float3Write=true
   )
   &ScaleSpriteSize(
    Uniform Curve Scale=0.8
    Uniform Curve Sprite Scale=DataInterface() {propertyValues {Curve {
       defaultValue=0.3
       keys [{interpMode=RCIM_Cubic, tangentMode=RCTM_User, time=0, value=0}, {interpMode=RCIM_Cubic, tangentMode=RCTM_User, time=1, value=1}]
      }}}
   )
   &InheritSourceMovement(
    Applied Position Scale=&VectorFromFloat(Value=&FloatFromCurve(FloatCurve=DataInterface() {propertyValues {Curve {
         defaultValue=1
         keys [
          {arriveTangent=0.016, interpMode=RCIM_Cubic, leaveTangent=0.016, tangentMode=RCTM_User, time=0, value=1}
          {arriveTangent=-3.0931, interpMode=RCIM_Cubic, leaveTangent=-3.0931, tangentMode=RCTM_User, time=1, value=0}
         ]
        }}}))
    ApplyForce=false
    ApplyPosition=true
    ApplyPositionFalloff=false
   )
   &SolveForcesAndVelocity(Clamp Velocity=false, Limit Acceleration=false)
   &ScaleColor(
    Scale Alpha=&FloatFromCurve(FloatCurve=DataInterface() {propertyValues {Curve {
        defaultValue=0
        keys [
         {interpMode=RCIM_Cubic, tangentMode=RCTM_User, time=0, value=0}
         {interpMode=RCIM_Cubic, tangentMode=RCTM_User, time=0.5, value=1}
         {tangentMode=RCTM_User, time=1, value=0}
        ]
       }}})
    Scale RGB=&MakeVectorFromLinearColorRGB(Color=&ColorFromCurve(ColorCurve=DataInterface() {propertyValues {
        AlphaCurve {keys [<at 0 0.6 RCIM_Cubic>,<at 1 0 RCIM_Cubic>]}
        BlueCurve {keys [<at 0 0.0089>, {interpMode=RCIM_Cubic, leaveTangent=-0.0035, time=0.3617, value=1}, <at 0.9059 0.0053>]}
        GreenCurve {keys [<at 0 0.0196>, {interpMode=RCIM_Cubic, leaveTangent=-0.0078, time=0.3617, value=1}, <at 0.9059 0.0118>]}
        RedCurve {keys [<at 0 0.0521>, {interpMode=RCIM_Cubic, leaveTangent=-0.0208, time=0.3617, value=1}, <at 0.9059 0.0312>]}
       }}))
    ScaleA=true
    ScaleRGB=true
   )
  ]
  rendererClasses [&NiagaraSpriteRendererProperties]
  renderers [{rendererClass=&NiagaraSpriteRendererProperties}]
 }
 Snow_wings {
  emitterUpdateScript [
   &EmitterState(Life Cycle Mode=::NewEnumerator1, Loop Behavior=::NewEnumerator1, Loop Duration=1, UseLoopDelay=false)
   &SpawnRate(SpawnRate=30, Use Spawn Probability=false) {module=SpawnRate001}
  ]
  particleSpawnScript [
   /SetVariables(Particles.VisibilityTag=&SpawnGroup()) {isSetParameters=true}
   &InitializeParticle(
    Color=<rgba 70 19.417 0 1>
    Lifetime Max=0.8
    Lifetime Min=0.5
    Lifetime Mode=::NewEnumerator1
    Sprite Rotation Mode=::NewEnumerator2
    Sprite Size Mode=::NewEnumerator1
    Uniform Sprite Size Max=8
    UsePositionOffset=false
    Write Material Random=true
   )
   &AddVelocity(Velocity=&Multiply_VectorByFloat(
     Vector=&RandomRangeVector(Minimum=<-500 -500 -500>, Maximum=<500 500 500>)
     Float=&FN_FrontEndFloatModifier(Gameplay=1, Front End=0.2)
    ))
   &SkeletalMeshLocation(
    Bone / Socket Index=&MergeHelper_IntSwitch(
     0=1
     1=2
     2=3
     3=4
     4=?
     5=?
     Index=&ModuloInt(A=&ReturnParticleIDAttributes(), B=2)
     Num=::NewEnumerator2
    )
    Bone Sampling Mode=::NewEnumerator4
    MaskSpawnGroup=false
    Orientation Sampling=::NewEnumerator1
    Sampled Scale Coordinate Space:ENiagaraCoordinateSpace=::Simulation
    Skeletal Mesh=$System.Skeletal Mesh Reader
   )
   &CameraOffset(Camera Offset Amount=5)
  ]
  particleUpdateScript [
   &ParticleState()
   &GravityForce(Gravity=&MakeVector(X=0, Y=0, Z=&RandomRangeFloat(Minimum=-75, Maximum=-150)))
   &Drag(Drag=&RandomRangeFloat(Minimum=1, Maximum=2), Use Linear Drag=true, Use Rotational Drag=false)
   &VectorNoiseForce(
    Force Amount=&Multiply_Float(
     A=&FloatFromCurve(
      FloatCurve=DataInterface() {propertyValues {Curve {keys [
          {arriveTangent=4.0849, interpMode=RCIM_Cubic, leaveTangent=4.0849, leaveTangentWeight=0.15, tangentMode=RCTM_User, time=0, value=0}
          {arriveTangentWeight=0.15, interpMode=RCIM_Cubic, tangentMode=RCTM_User, time=0.45, value=1}
         ]}}}
      Scale Curve=500
     )
     B=1
    )
    Module.Use Random Noise Time Interval=false
   )
   &InheritSourceMovement(Applied Position Scale=<0.8 0.8 0.5>, ApplyForce=false, ApplyPosition=true, ApplyPositionFalloff=false)
   &SolveForcesAndVelocity(Clamp Velocity=false, Limit Acceleration=false)
   &ScaleColor(
    Scale Alpha=&FloatFromCurve(FloatCurve=DataInterface() {propertyValues {Curve {defaultValue=1, keys [{tangentMode=RCTM_User, time=0, value=0}, {tangentMode=RCTM_User, time=1, value=1}]}}})
    Scale RGB=&MakeVectorFromLinearColorRGB(Color=&ColorFromCurve(ColorCurve=DataInterface() {propertyValues {
        AlphaCurve {keys [<at 0 1 RCIM_Cubic>,<at 1 0 RCIM_Cubic>]}
        BlueCurve {keys [<at 0 5.7076 RCIM_Cubic>,<at 0.7322 0.2272>,<at 1 0 RCIM_Cubic>]}
        GreenCurve {keys [<at 0 64.7427 RCIM_Cubic>,<at 0.7322 0.6189>,<at 1 0 RCIM_Cubic>]}
        RedCurve {keys [<at 0 100 RCIM_Cubic>,<at 0.7322 3.981>,<at 1 0 RCIM_Cubic>]}
       }}))
    ScaleA=true
    ScaleRGB=true
   ) {module=ScaleColor001}
   &SpriteRotationRate(Rotation Rate=&RandomRangeFloat(Minimum=-0.4, Maximum=0.4))
   &ScaleSpriteSize(Uniform Curve Sprite Scale=DataInterface() {propertyValues {Curve {keys [
        <at 0 0 RCIM_Cubic>
        {arriveTangent=3.2789, interpMode=RCIM_Cubic, leaveTangent=3.2789, time=0.15, value=1.142}
        {arriveTangent=-1.3435, interpMode=RCIM_Cubic, leaveTangent=-1.3435, time=0.42, value=1.3772}
        <at 1 0 RCIM_Cubic>
       ]}}})
   &DynamicMaterialParameters(
    Index 0 Param 1=&FloatFromCurve(FloatCurve=DataInterface() {propertyValues {Curve {keys [
         {interpMode=RCIM_Cubic, leaveTangentWeight=0.15, tangentMode=RCTM_User, time=0, value=1.2}
         {
          arriveTangent=-2.0203
          arriveTangentWeight=0.3808
          interpMode=RCIM_Cubic
          leaveTangent=-2.0203
          leaveTangentWeight=0.1269
          tangentMode=RCTM_User
          time=0.45
          value=0
         }
         {arriveTangentWeight=0.1038, interpMode=RCIM_Cubic, leaveTangentWeight=0.2767, tangentMode=RCTM_User, time=0.6, value=-0.2}
         {arriveTangentWeight=0.1333, interpMode=RCIM_Cubic, tangentMode=RCTM_User, time=1, value=1}
        ]}}})
    Index 0 Param 2=&FloatFromCurve(FloatCurve=DataInterface() {propertyValues {Curve {keys [
         {interpMode=RCIM_Cubic, leaveTangentWeight=0.2833, tangentMode=RCTM_User, time=0, value=3.5}
         {arriveTangentWeight=1.0313, interpMode=RCIM_Cubic, leaveTangentWeight=0.182, tangentMode=RCTM_User, time=0.85, value=2.8}
         {arriveTangentWeight=0.05, interpMode=RCIM_Cubic, tangentMode=RCTM_User, time=1, value=0}
        ]}}})
    Index 0 Param 3=&FloatFromCurve(FloatCurve=DataInterface() {propertyValues {Curve {keys [
         {interpMode=RCIM_Cubic, leaveTangentWeight=0.1, tangentMode=RCTM_User, time=0.2, value=0}
         {arriveTangentWeight=0.1, interpMode=RCIM_Cubic, leaveTangentWeight=0.1267, tangentMode=RCTM_User, time=0.5, value=4}
         {arriveTangentWeight=0.1267, interpMode=RCIM_Cubic, tangentMode=RCTM_User, time=0.88, value=0}
        ]}}})
    Index 0 Param 4=&RandomRangeFloat(Minimum=0, Maximum=4)
    Module.Param0WriteEnabled=true
    Module.Param1WriteEnabled=true
    Module.Param2WriteEnabled=true
    Module.Param3WriteEnabled=true
   )
  ]
  rendererClasses [&NiagaraSpriteRendererProperties]
  renderers [{rendererClass=&NiagaraSpriteRendererProperties}]
 }
 Wings_Trail {
  emitterSpawnScript [&FN_RibbonIDOffsetUtility()]
  emitterUpdateScript [
   &EmitterState(
    Inactive Response=::NewEnumerator1
    Life Cycle Mode=::NewEnumerator1
    Loop Behavior=::NewEnumerator1
    Loop Delay=1
    Loop Duration=0.9
    UseLoopDelay=false
   )
   &FN_SpawnAnimTrail(Number of Anim Trails=2, Switch on ENiagara_AnimTrail_Source:ENiagara_AnimTrail_Source=::NewEnumerator1)
  ]
  particleSpawnScript [
   &FN_InitializeAnimTrailBoneIndecies(Second Bone or Socket Index=2)
   &FN_InitializeAnimTrailBoneIndecies(AnimTrail Selection=1, First Bone or Socket Index=1, Second Bone or Socket Index=2) {module=FN_InitializeAnimTrailBoneIndecies001}
   &FN_InitializeAnimTrail(Color=<rgba 0 0 0 2>, SkeletalMesh=$System.Skeletal Mesh Reader)
   &CameraOffset()
  ]
  particleUpdateScript [
   &InheritSourceMovement(
    Applied Position Scale=&Multiply_VectorByFloat(
     Vector=<1 1 1>
     Float=&FloatFromCurve(FloatCurve=DataInterface() {propertyValues {Curve {
         defaultValue=1
         keys [
          {arriveTangent=-0.0065, interpMode=RCIM_Cubic, leaveTangent=-0.0065, tangentMode=RCTM_User, time=0, value=0}
          {arriveTangent=4.1036, interpMode=RCIM_Cubic, leaveTangent=4.1036, tangentMode=RCTM_User, time=1, value=1}
         ]
        }}})
    )
    ApplyForce=false
    ApplyPosition=true
    ApplyPositionFalloff=false
   )
   &SolveForcesAndVelocity(Clamp Velocity=false, Limit Acceleration=false)
   &FN_UpdateAnimTrail(SkeletalMesh=$System.Skeletal Mesh Reader) {module=FN_UpdateAnimTrail001}
   &DynamicMaterialParameters(
    Index 0 Param 1=0.2
    Module.Index0Float0Write=true
    Module.Index0Float1Write=true
    Module.Index0Float2Write=true
    Module.Index0Float3Write=true
   )
   &ScaleColor(
    Scale Alpha=&Multiply_Float(
     A=&FloatFromCurve(
      FloatCurve=DataInterface() {propertyValues {Curve {keys [
          {arriveTangent=-1.5186, interpMode=RCIM_Cubic, leaveTangent=-1.5186, tangentMode=RCTM_User, time=0, value=0.8}
          {arriveTangent=-1.4453, interpMode=RCIM_Cubic, leaveTangent=-1.4453, tangentMode=RCTM_User, time=0.2393, value=0.25}
          {interpMode=RCIM_Cubic, tangentMode=RCTM_User, time=1, value=0}
         ]}}}
      Scale Curve=2
     )
     B=&FN_FrontEndFloatModifier(
      Gameplay=1
      Front End=&FloatFromCurve(FloatCurve=DataInterface() {propertyValues {Curve {keys [<at 1 0>,<at 3 1>]}}}, CurveIndex=$Emitter.Age)
     )
    )
    ScaleA=true
    ScaleRGB=true
   ) {module=ScaleColor001}
   &ScaleColor(
    Scale Alpha=&Multiply_Float(A=$System.FqdeIn, B=1)
    Scale RGB=&MakeVectorFromLinearColorRGB(Color=&FNTimeOfDayColor(
      Night Color=<rgba 0.451 0.5602 0.849 1>
      Morning Color=<rgba 1 0.9059 0.651 1>
      Day Color=<rgba 1 1 1 1>
      Evening Color=<rgba 1 0.849 0.8962 1>
     ))
    ScaleA=true
    ScaleRGB=true
   ) {module=ScaleColor002}
  ]
  rendererClasses [&NiagaraRibbonRendererProperties]
  renderers [{rendererClass=&NiagaraRibbonRendererProperties}]
 }
}
summary {
 systemName=NS_Glider_DevilRock_Deploy
 emitters [
  {emitterName=Fire_Chest, rendererClasses [&NiagaraSpriteRendererProperties]}
  {emitterName=Wings_Trail, rendererClasses [&NiagaraRibbonRendererProperties]}
  {emitterName=Snow_wings, rendererClasses [&NiagaraSpriteRendererProperties]}
  {emitterName=Fire_DeployTrail, rendererClasses [&NiagaraSpriteRendererProperties]}
 ]
 userVariables [
  User.bIsFrontEnd {type=NiagaraBool}
  User.bIsFrontEndPreview {type=NiagaraBool}
  User.RibbonIDOffset {type=NiagaraInt32}
 ]
}
