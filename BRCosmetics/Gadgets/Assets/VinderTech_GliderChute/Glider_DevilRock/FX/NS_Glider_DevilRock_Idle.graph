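# Niagara system for the idle effect of the DevilRock glider cosmetic, shedding fire, smoke, glowing ash and flame decals from the glider.
# Seven sprite emitters place particles at indexed skeletal mesh bones or sockets with per-group offsets, with turbulence vector fields moving the ash and spark particles.
# User.ForwardDot scales several spawn rates, radius and smoke velocity, while front-end modifiers retune gameplay versus menu looks and time-of-day scalars dim some emitters' alpha.
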NiagaraSystem /BRCosmetics/Gadgets/Assets/VinderTech_GliderChute/Glider_DevilRock/FX/NS_Glider_DevilRock_Idle
dependencies {
 usedDataInterfaces [
  &NiagaraDataInterfaceArrayInt32
  &NiagaraDataInterfaceArrayFloat3
  /Script/Niagara.NiagaraDataInterfaceVectorField
  &NiagaraDataInterfaceCurve
  &NiagaraDataInterfaceSpriteRendererInfo
  &NiagaraDataInterfaceColorCurve
  &NiagaraDataInterfaceVector2DCurve
  &NiagaraDataInterfaceSkeletalMesh
  &NiagaraDataInterfaceSocketReader
 ]
 usedDynamicInputs [&Boolean_AndOperation, &SetBoolByFloatComparison]
 usedModules [
  &FN_VariantManager
  &self:SystemSpawnScript.SystemScriptSource.SystemScriptGraph.NiagaraNodeAssignment_0.SetVariables_#1
  &self:SystemSpawnScript.SystemScriptSource.SystemScriptGraph.NiagaraNodeAssignment_1.SetVariables_#2
  &SystemState
  &Timeline
  &self:SystemSpawnScript.SystemScriptSource.SystemScriptGraph.NiagaraNodeAssignment_2.SetVariables_#3
  &FN_Integrator
 ]
}
emitters {
 Ash_Wings {
  emitterUpdateScript [
   &EmitterState()
   &SpawnRate(SpawnRate=10, Use Spawn Probability=false)
  ]
  particleSpawnScript [
   &FN_SetSpawnGroup(Number of Spawn Groups=2)
   &InitializeParticle(
    Color=<rgba 1.5 0.6292 0.0452 1>
    Lifetime Max=3
    Lifetime Min=2
    Lifetime Mode=::NewEnumerator1
    Sprite Rotation Mode=::NewEnumerator2
    Sprite Size Mode=::NewEnumerator1
    Uniform Sprite Size Max=3
    Uniform Sprite Size Min=1.2
    UsePositionOffset=false
    Write Material Random=true
   )
   &AddVelocity(Velocity=&RandomRangeVector(Minimum=<50 10 50>, Maximum=<-50 5 30>))
   &SkeletalMeshLocation(
    Bone / Socket Index=&SelectIntFromArray(Direct Array Index=$Particles.SpawnGroup, Int Selection Array=DataInterface() {propertyValues {IntData [2,3]}})
    Bone Sampling Mode=::NewEnumerator4
    MaskSpawnGroup=false
    Position Sampling=::NewEnumerator1
    Sampled Position Offset=&AddVectorToPosition(
     Position=&SelectVectorFromArray(
      Direct Array Index=&SpawnGroup()
      Vector Selection Array=DataInterface() {propertyValues {FloatData [<80 0 0>, <-80 0 0>]}}
     )
     Vector=&FN_FrontEndVector3Modifier(
      Gameplay=&RandomRangeVector(Minimum=<-30 -100 -10>, Maximum=<30 100 10>)
      Front End=&RandomRangeVector(Minimum=<-30 -70 -10>, Maximum=<30 70 10>)
     )
    )
    Skeletal Mesh=$System.Skeletal Mesh Reader
   )
   &CameraOffset(Camera Offset Amount=10)
  ]
  particleUpdateScript [
   &ParticleState()
   /Niagara/VectorFields/SampleVectorField(
    Sampled Vector Scale=0.4
    VectorField=DataInterface() {propertyValues {Field=/Game/Effects/VectorFields/Noisy_Turbulence_103, bTileX=true, bTileY=true, bTileZ=true}}
   )
   /Niagara/VectorFields/ApplyVectorField(Vector Field Intensity as Velocity=10, Vector Field Velocity Tightness=0.05)
   &ScaleSpriteSize(Uniform Curve Sprite Scale=DataInterface() {propertyValues {Curve {keys [
        {interpMode=RCIM_Cubic, tangentMode=RCTM_User, time=0, value=0}
        {interpMode=RCIM_Cubic, tangentMode=RCTM_User, time=0.15, value=1}
        {arriveTangent=-1.4402, interpMode=RCIM_Cubic, leaveTangent=-1.4402, tangentMode=RCTM_User, time=0.65, value=0.25}
        {interpMode=RCIM_Cubic, tangentMode=RCTM_User, time=1, value=0}
       ]}}})
   &SubUVAnimation(
    "Sprite Renderer "=DataInterface() {propertyValues {SpriteRenderer=&self:Ash_Wings_0.NiagaraSpriteRendererProperties_0}}
    UseEndFrame=false
    UseStartFrame=false
   )
   &DynamicMaterialParameters(
    Index 0 Param 1=&RandomRangeFloat(Minimum=0, Maximum=1)
    Index 0 Param 2=&RandomRangeFloat(Minimum=0, Maximum=1)
    Index 0 Param 3=&RandomRangeFloat(Minimum=0.2, Maximum=0.5)
    Index 0 Param 4=&RandomRangeFloat(Minimum=0.5, Maximum=2)
    Module.Index0Float0Write=true
    Module.Index0Float1Write=true
    Module.Index0Float2Write=true
    Module.Index0Float3Write=true
   )
   &ScaleColor(
    Scale Alpha=&FloatFromCurve(FloatCurve=DataInterface() {propertyValues {Curve {
        defaultValue=1
        keys [{interpMode=RCIM_Cubic, tangentMode=RCTM_User, time=0, value=1}, {interpMode=RCIM_Cubic, tangentMode=RCTM_User, time=1, value=0}]
       }}})
    Scale RGB=&MakeVectorFromLinearColorRGB(Color=&ColorFromCurve(ColorCurve=DataInterface() {propertyValues {
        AlphaCurve {keys [<at 0 1 RCIM_Cubic>,<at 1 0 RCIM_Cubic>]}
        BlueCurve {keys [<at 0 2.8538 RCIM_Cubic>,<at 0.7322 0.2272>,<at 1 0 RCIM_Cubic>]}
        GreenCurve {keys [<at 0 32.3713 RCIM_Cubic>,<at 0.7322 0.6189>,<at 1 0 RCIM_Cubic>]}
        RedCurve {keys [<at 0 50 RCIM_Cubic>,<at 0.7322 3.981>,<at 1 0 RCIM_Cubic>]}
       }}))
    ScaleA=true
    ScaleRGB=true
   )
   &InheritSourceMovement(Applied Position Scale=<0.8 0.8 1>, ApplyForce=false, ApplyPosition=true, ApplyPositionFalloff=false)
   &SolveForcesAndVelocity(Clamp Velocity=false, Limit Acceleration=false)
  ]
  rendererClasses [&NiagaraSpriteRendererProperties]
  renderers [{rendererClass=&NiagaraSpriteRendererProperties}]
 }
 CPU_particles {
  emitterUpdateScript [
   &EmitterState()
   &SpawnRate(
    SpawnRate=&Multiply_Float(
     A=15
     B=&FN_FrontEndFloatModifier(Gameplay=&ClampFloat(Float=$User.ForwardDot, Min=0.8), Front End=1)
    )
    Use Spawn Probability=false
   )
  ]
  particleSpawnScript [
   &FN_SetSpawnGroup(Normalize Spawn Group=true, Particles Per Group=5, Set Ribbon ID and Ribbon Link Order=true)
   &InitializeParticle(
    Color=<rgba 0.5 0.2114 0.0052 1>
    Lifetime Max=1.2
    Lifetime Min=0.7
    Lifetime Mode=::NewEnumerator1
    Sprite Rotation Mode=::NewEnumerator2
    Sprite Size Mode=::NewEnumerator3
    Uniform Sprite Size=&Multiply_Float(
     A=&RandomRangeFloat(Minimum=5, Maximum=15)
     B=&MergeHelper_FloatSwitch(0=0.3, 1=0.3, 2=0.4, 3=0.5, Index=$Particles.SpawnGroup, Num=::NewEnumerator2)
    )
    UsePositionOffset=false
    Write Material Random=true
   )
   &SkeletalMeshLocation(
    Bone / Socket Index=&MergeHelper_IntSwitch(0=0, 1=1, 2=14, 3=12, 4=?, 5=?, Index=$Particles.SpawnGroup, Num=::NewEnumerator2)
    Bone Sampling Mode=::NewEnumerator4
    Interpolate Mesh Sampling=false
    MaskSpawnGroup=false
    Orientation Sampling=::NewEnumerator1
    Sampled Position Offset=&AddVectorToPosition(
     Position=&MergeHelper_VectorSwitch(0=<-15 0 0>, 1=<15 0 0>, 2=<40 0 0>, 3=<0 0 20>, Index=$Particles.SpawnGroup, Num=::NewEnumerator2)
     Vector=&MergeHelper_VectorSwitch(
      0=<0 0 0>
      1=<0 0 0>
      2=<0 0 0>
      3=&RandomRangeVector(Minimum=<0 0 0>, Maximum=<-110 0 20>)
      Index=$Particles.SpawnGroup
      Num=::NewEnumerator2
     )
    )
    Skeletal Mesh=$System.Skeletal Mesh Reader
   )
   &SphereLocation(
    MaskSpawnGroup=false
    Non Uniform Scale=<1 1 0.7>
    Override Seed=false
    Sphere Radius=&Multiply_Float(A=30, B=&ClampFloat(Float=$User.ForwardDot, Min=0.8))
    Surface Only=false
    UseNonUniformScale=true
    UseOffset=false
    UseSphereOrientation=false
   )
   &CameraOffset(Camera Offset Amount=10)
   /SetVariables(Particles.VisibilityTag=&RandomRangeInt(Minimum=0, Maximum=1)) {isSetParameters=true}
  ]
  particleUpdateScript [
   &ParticleState()
   &InheritSourceMovement(Applied Position Scale=<0.9 0.9 1>, ApplyForce=false, ApplyPosition=true, ApplyPositionFalloff=false)
   &ScaleSpriteSize(Scale Factor=&Vector2DFromFloat(Value=&FloatFromCurve(FloatCurve=DataInterface() {propertyValues {Curve {keys [<at 0 0 RCIM_Cubic>,<at 0.1351 0.878 RCIM_Cubic>,<at 1 0 RCIM_Cubic>]}}}))) {module=ScaleSpriteSize001}
   &SpriteRotationRate(Rotation Rate=&RandomRangeFloat(Minimum=-1, Maximum=1))
   &ScaleColor(
    Scale Alpha=&Multiply_Float(
     A=&RandomRangeFloat(Minimum=0.2, Maximum=1)
     B=&FloatFromCurve(
      FloatCurve=DataInterface() {propertyValues {Curve {keys [
          {arriveTangent=4.1785, interpMode=RCIM_Cubic, leaveTangent=4.1785, tangentMode=RCTM_User, time=0, value=0}
          {arriveTangent=0.9995, interpMode=RCIM_Cubic, leaveTangent=0.9995, tangentMode=RCTM_User, time=0.2223, value=0.6825}
          {arriveTangent=-0.0405, interpMode=RCIM_Cubic, leaveTangent=-0.0405, tangentMode=RCTM_User, time=1, value=1}
         ]}}}
      Scale Curve=&FNTimeOfDayScalar(Night Scale=1, Morning Scale=0.2, Evening Scale=0.4)
     )
    )
    Scale RGB=&MakeVectorFromLinearColorRGB(Color=&ColorFromCurve(ColorCurve=DataInterface() {propertyValues {
        AlphaCurve {keys [<at 0 0 RCIM_Cubic>,<at 1 0.7 RCIM_Cubic>]}
        BlueCurve {keys [<at 0 6000 RCIM_Cubic>,<at 1 3125 RCIM_Cubic>]}
        GreenCurve {keys [<at 0 6000 RCIM_Cubic>,<at 1 3169.7529 RCIM_Cubic>]}
        RedCurve {keys [<at 0 6000 RCIM_Cubic>,<at 1 5000 RCIM_Cubic>]}
       }}))
    ScaleA=true
    ScaleRGB=true
   )
   &SampleVectorField(VectorField=DataInterface() {propertyValues {Field=&Noisy_Turbulence_103, bTileX=true, bTileY=true, bTileZ=true}})
   &ApplyVectorField(Vector Field Intensity as Velocity=10, Vector Field Velocity Tightness=0.2)
   &SolveForcesAndVelocity(Clamp Velocity=false, Limit Acceleration=false)
  ]
  rendererClasses [&NiagaraSpriteRendererProperties]
  renderers [{rendererClass=&NiagaraSpriteRendererProperties}]
 }
 Dark_Smoke_Wings {
  emitterUpdateScript [
   &EmitterState()
   &SpawnRate(SpawnRate=5, Use Spawn Probability=false)
  ]
  particleSpawnScript [
   &PartitionParticles(Number Of Partitions=6)
   &InitializeParticle(
    Color=<rgba 0.2031 0.1309 0.1268 1.5>
    Lifetime Min=1.4
    Lifetime Mode=::NewEnumerator1
    Sprite Rotation Mode=::NewEnumerator2
    Sprite Size Max=<60 200>
    Sprite Size Min=<50 150>
    Sprite Size Mode=::NewEnumerator2
    UsePositionOffset=false
    Write Material Random=true
   )
   &AddVelocity(Velocity=&RandomRangeVector(Minimum=<0 90 0>, Maximum=<0 110 0>))
   &CameraOffset(Camera Offset Amount=30)
  ]
  particleUpdateScript [
   &ParticleState()
   &SkeletalMeshLocation(
    Bone / Socket Index=&SelectIntFromArray(Direct Array Index=$Particles.PartitionParticles.Partition, Int Selection Array=DataInterface() {propertyValues {IntData [2,3,4,5,6,7]}})
    Bone Sampling Mode=::NewEnumerator4
    MaskSpawnGroup=false
    Sampled Position Offset=&SelectVectorFromArray(
     Direct Array Index=$Particles.PartitionParticles.Partition
     Vector Selection Array=DataInterface() {propertyValues {FloatData [<25 -17 0>, <-25 17 0>, <37 -37 0>, <-37 37 0>, <-35 0 0>, <35 0 0>]}}
    )
    Skeletal Mesh=$System.Skeletal Mesh Reader
   )
   &ScaleSpriteSize(
    Non-Uniform Curve Sprite Scale=DataInterface() {propertyValues {
      XCurve {keys [{tangentMode=RCTM_User, time=0, value=0.6}, {tangentMode=RCTM_User, time=0.4, value=1}, <at 1 0.3>]}
      YCurve {keys [{tangentMode=RCTM_User, time=0, value=0.5}, <at 0.4 0.7>, {tangentMode=RCTM_User, time=1, value=1.1}]}
     }}
    Scale Sprite Size Mode=::NewEnumerator3
   )
   &DynamicMaterialParameters(
    Index 0 Param 1=&RandomRangeFloat(Minimum=0, Maximum=1)
    Index 0 Param 2=&RandomRangeFloat(Minimum=0.52, Maximum=0.7)
    Index 0 Param 3=&RandomRangeFloat(Minimum=0.55, Maximum=0.75)
    Index 0 Param 4=&FloatFromCurve(FloatCurve=DataInterface() {propertyValues {Curve {keys [<at 0 2>,<at 0.7 1>]}}})
    Module.Index0Float0Write=true
    Module.Index0Float1Write=true
    Module.Index0Float2Write=true
    Module.Index0Float3Write=true
   )
   &ScaleColor(
    Scale Alpha=&FloatFromCurve(
     FloatCurve=DataInterface() {propertyValues {Curve {keys [
         {interpMode=RCIM_Cubic, tangentMode=RCTM_User, time=0, value=0}
         {arriveTangent=2.0657, interpMode=RCIM_Cubic, leaveTangent=2.0657, tangentMode=RCTM_User, time=0.3, value=1}
         {interpMode=RCIM_Cubic, tangentMode=RCTM_User, time=0.3, value=1}
         <at 1 0>
        ]}}}
     Scale Curve=&FN_FrontEndFloatModifier(Gameplay=1, Front End=0.2)
    )
    Scale RGB=&MakeVectorFromLinearColorRGB(Color=&ColorFromCurve(ColorCurve=DataInterface() {propertyValues {
        AlphaCurve {keys [<at 0 0 RCIM_Cubic>,<at 1 0.6 RCIM_Cubic>]}
        BlueCurve {keys [<at 0 0.3586 RCIM_Cubic>,<at 1 0.1719 RCIM_Cubic>]}
        GreenCurve {keys [<at 0 1.3373 RCIM_Cubic>,<at 1 0.1684 RCIM_Cubic>]}
        RedCurve {keys [<at 0 7 RCIM_Cubic>,<at 1 0.1588 RCIM_Cubic>]}
       }}))
    ScaleA=true
    ScaleRGB=true
   )
   &SolveForcesAndVelocity(Clamp Velocity=false, Limit Acceleration=false)
  ]
  rendererClasses [&NiagaraSpriteRendererProperties]
  renderers [{rendererClass=&NiagaraSpriteRendererProperties}]
 }
 Fire {
  emitterUpdateScript [
   &EmitterState(Life Cycle Mode=::NewEnumerator1, Loop Behavior=::NewEnumerator1, Loop Delay=1.2, Loop Duration Mode=::NewEnumerator1, UseLoopDelay=true)
   &SpawnRate(Spawn Probability=0.75, SpawnRate=15, Use Spawn Probability=true)
  ]
  particleSpawnScript [
   &FN_SetSpawnGroup(Normalize Spawn Group=true, Number of Spawn Groups=5, Particles Per Group=1, Set Ribbon ID and Ribbon Link Order=true)
   &InitializeParticle(
    Color=&MergeHelper_ColorSwitch(
     0=<rgba 0.0114 0.0156 0.0208 0.6>
     1=<rgba 0.0114 0.0156 0.0208 0.6>
     2=<rgba 0.0114 0.0156 0.0208 0.6>
     3=?
     Index=$Particles.SpawnGroup
     Num=::NewEnumerator1
    )
    Lifetime Max=1.4
    Lifetime Min=0.8
    Lifetime Mode=::NewEnumerator1
    Sprite Size=&Multiply_Vector2D_ByFloat(
     Float=&FN_FrontEndFloatModifier(
      Gameplay=&MergeHelper_FloatSwitch(0=0.6, 1=0.6, 2=0.8, 3=?, Index=$Particles.SpawnGroup, Num=::NewEnumerator1)
      Front End=&MergeHelper_FloatSwitch(0=0.5, 1=0.5, 2=0.8, 3=?, Index=$Particles.SpawnGroup, Num=::NewEnumerator1)
     )
     Vector2D=&RandomRangeVector2D(Minimum=<60 60>, Maximum=<80 80>)
    )
    Sprite Size Mode=::NewEnumerator4
    UsePositionOffset=false
    Write Material Random=true
   )
   &CameraOffset(Camera Offset Amount=&FN_FrontEndFloatModifier(Gameplay=30, Front End=0))
   &DynamicMaterialParameters(
    Index 0 Param 1=1
    Index 0 Param 2=1
    Index 0 Param 3=1
    Index 0 Param 4=1
    Module.Param0WriteEnabled=true
    Module.Param1WriteEnabled=true
    Module.Param2WriteEnabled=true
    Module.Param3WriteEnabled=true
   )
  ]
  particleUpdateScript [
   &ParticleState()
   &SkeletalMeshLocation(
    Bone / Socket Index=&MergeHelper_IntSwitch(0=0, 1=1, 2=14, 3=?, 4=?, 5=?, Index=$Particles.SpawnGroup, Num=::NewEnumerator1)
    Bone Sampling Mode=::NewEnumerator4
    MaskSpawnGroup=false
    Orientation Sampling=::NewEnumerator1
    Sampled Position Offset=&MergeHelper_VectorSwitch(
     0=<-18 5 0>
     1=<18 -5 0>
     2=&AddVectorToPosition(Position=<40 0 -11.1443>, Vector=&RandomRangeVector(Minimum=<-5 0 0>, Maximum=<20 1 1>))
     3=?
     Index=$Particles.SpawnGroup
     Num=::NewEnumerator1
    )
    Skeletal Mesh=$System.Skeletal Mesh Reader
   ) {enabled=false}
   &SkeletalMeshLocation(
    Bone / Socket Index=&MergeHelper_IntSwitch(0=0, 1=1, 2=14, 3=?, 4=?, 5=?, Index=$Particles.SpawnGroup, Num=::NewEnumerator1)
    Bone Sampling Mode=::NewEnumerator4
    MaskSpawnGroup=false
    Orientation Sampling=::NewEnumerator1
    Sampled Position Offset=&MergeHelper_VectorSwitch(
     0=<-18 0 0>
     1=<18 0 0>
     2=&AddVectorToPosition(Position=<40 0 -11.1443>, Vector=&RandomRangeVector(Minimum=<-5 0 0>, Maximum=<20 1 1>))
     3=?
     Index=$Particles.SpawnGroup
     Num=::NewEnumerator1
    )
    Skeletal Mesh=$System.Skeletal Mesh Reader
   ) {module=SkeletalMeshLocation001}
   &ScaleSpriteSize(Uniform Curve Sprite Scale=DataInterface() {propertyValues {Curve {
       defaultValue=0
       keys [{interpMode=RCIM_Cubic, tangentMode=RCTM_User, time=0, value=0}, {interpMode=RCIM_Cubic, tangentMode=RCTM_User, time=1, value=1}]
      }}})
   /SetVariables(Particles.FireColor:LinearColor=&MakeLinearColorFromVectorAndFloat(
     "Vector (RGB)"=&MakeVectorFromLinearColorRGB(Color=&ColorFromCurve(ColorCurve=DataInterface() {propertyValues {
         AlphaCurve {keys [<at 0 1 RCIM_Cubic>,<at 1 0 RCIM_Cubic>]}
         BlueCurve {keys [
           <at 0 0.0053>
           <at 0.2685 1.1641>
           {arriveTangent=-1.5841, interpMode=RCIM_Cubic, leaveTangent=-1.5841, time=0.3803, value=0.776}
           <at 1 0.0053 RCIM_Cubic>
          ]}
         GreenCurve {keys [
           <at 0 0.0231>
           <at 0.2685 1.1641>
           {arriveTangent=-1.5598, interpMode=RCIM_Cubic, leaveTangent=-1.5598, time=0.3803, value=0.776}
           <at 1 0.0231 RCIM_Cubic>
          ]}
         RedCurve {keys [
           <at 0 0.0833>
           <at 0.2685 1.5>
           {arriveTangent=-1.9367, interpMode=RCIM_Cubic, leaveTangent=-1.9367, time=0.3803, value=1}
           <at 1 0.0833 RCIM_Cubic>
          ]}
        }}))
     "Float (Alpha)"=&FloatFromCurve(FloatCurve=DataInterface() {propertyValues {Curve {
         defaultValue=1
         keys [
          {interpMode=RCIM_Cubic, tangentMode=RCTM_User, time=0, value=0}
          {interpMode=RCIM_Cubic, tangentMode=RCTM_User, time=0.5, value=1}
          {tangentMode=RCTM_User, time=1, value=0}
         ]
        }}})
    )) {isSetParameters=true, module=SetVariables_#4}
  ]
  rendererClasses [&NiagaraSpriteRendererProperties]
  renderers [{rendererClass=&NiagaraSpriteRendererProperties}]
 }
 Fire_and_Smoke {
  emitterUpdateScript [
   &EmitterState(Life Cycle Mode=::NewEnumerator1, Loop Behavior=::NewEnumerator1, Loop Delay=1.2, Loop Duration Mode=::NewEnumerator1, UseLoopDelay=true)
   &SpawnRate(
    Spawn Probability=0.75
    SpawnRate=&Multiply_Float(
     A=30
     B=&FN_FrontEndFloatModifier(Gameplay=&ClampFloat(Float=$User.ForwardDot, Min=0.7), Front End=1)
    )
    Use Spawn Probability=true
   )
  ]
  particleSpawnScript [
   &FN_SetSpawnGroup(Normalize Spawn Group=true, Number of Spawn Groups=3, Particles Per Group=10, Set Ribbon ID and Ribbon Link Order=true)
   &InitializeParticle(
    Color=&MergeHelper_ColorSwitch(
     0=<rgba 0.0729 0.0239 0.011 0.5>
     1=<rgba 0.0729 0.0239 0.011 0.5>
     2=<rgba 0.0729 0.0239 0.011 0.5>
     3=?
     Index=$Particles.SpawnGroup
     Num=::NewEnumerator1
    )
    Lifetime Max=1.2
    Lifetime Min=0.7
    Lifetime Mode=::NewEnumerator1
    Sprite Rotation Mode=::NewEnumerator2
    Sprite Size Mode=::NewEnumerator3
    Uniform Sprite Size=&Multiply_Float(
     A=&RandomRangeFloat(Minimum=70, Maximum=90)
     B=&MergeHelper_FloatSwitch(0=0.6, 1=0.6, 2=1, 3=?, Index=$Particles.SpawnGroup, Num=::NewEnumerator1)
    )
    UsePositionOffset=false
    Write Material Random=true
   )
   &AddVelocityFromPoint(
    Module.Use Falloff Distance=false
    Origin Offset=&Multiply_VectorByFloat(Vector=<0 -1 0>, Float=&ClampFloat(Float=$User.ForwardDot))
    Velocity Strength=&Lerp_Float(A=0, B=100, Alpha=$User.ForwardDot)
   )
   &AddVelocity(Velocity=&Multiply_VectorByFloat(
     Vector=&RandomRangeVector(Minimum=<0 0 0>, Maximum=<0 -100 0>)
     Float=&FN_FrontEndFloatModifier(Gameplay=0, Front End=1)
    ))
   &CameraOffset()
   &SkeletalMeshLocation(
    Bone / Socket Index=&MergeHelper_IntSwitch(0=0, 1=1, 2=14, 3=?, 4=?, 5=?, Index=$Particles.SpawnGroup, Num=::NewEnumerator1)
    Bone Sampling Mode=::NewEnumerator4
    Interpolate Mesh Sampling=false
    MaskSpawnGroup=false
    Orientation Sampling=::NewEnumerator1
    Sampled Position Offset=&MergeHelper_VectorSwitch(0=<-10 0 0>, 1=<10 0 0>, 2=<40 0 0>, 3=?, Index=$Particles.SpawnGroup, Num=::NewEnumerator1)
    Skeletal Mesh=$System.Skeletal Mesh Reader
   )
   &DynamicMaterialParameters(
    Index 0 Param 1=1
    Index 0 Param 2=1
    Index 0 Param 3=1
    Index 0 Param 4=1
    Module.Param0WriteEnabled=true
    Module.Param1WriteEnabled=true
    Module.Param2WriteEnabled=true
    Module.Param3WriteEnabled=true
   )
  ]
  particleUpdateScript [
   &ParticleState()
   &AccelerationForce(Acceleration=&MakeVector(X=0, Y=0, Z=300))
   &ScaleVelocity(Velocity Scale=&Multiply_VectorByFloat(
     Vector=<0.9 0.9 0.9>
     Float=&FN_FrontEndFloatModifier(Gameplay=&ClampFloat(Float=$User.ForwardDot, Min=0.5), Front End=1)
    ))
   &SolveForcesAndVelocity(Clamp Velocity=false, Limit Acceleration=false)
   &ScaleColor(
    Scale Alpha=&Multiply_Float(
     A=&FloatFromCurve(
      FloatCurve=DataInterface() {propertyValues {Curve {keys [
          {interpMode=RCIM_Cubic, tangentMode=RCTM_User, time=0, value=0}
          {interpMode=RCIM_Cubic, leaveTangent=-3.2414, tangentMode=RCTM_Break, time=0.2454, value=1}
          {tangentMode=RCTM_User, time=1, value=0}
         ]}}}
      Scale Curve=&Multiply_Float(
       A=&RandomRangeFloat(Minimum=0.5, Maximum=1)
       B=&FN_FrontEndFloatModifier(Gameplay=0.8, Front End=0.3)
      )
     )
     B=&FNTimeOfDayScalar(Night Scale=1, Morning Scale=0.2, Evening Scale=0.2)
    )
    Scale RGB=&MakeVectorFromLinearColorRGB(Color=&ColorFromCurve(ColorCurve=DataInterface() {propertyValues {
        AlphaCurve {keys [<at 0 1 RCIM_Cubic>,<at 1 0 RCIM_Cubic>]}
        BlueCurve {keys [<at 0 1 RCIM_Cubic>,<at 0.4682 0.0526>,<at 0.7505 0.024>,<at 1 0.0104>]}
        GreenCurve {keys [<at 0 1 RCIM_Cubic>,<at 0.4682 0.0526>,<at 0.7505 0.024>,<at 1 0.0104>]}
        RedCurve {keys [<at 0 1 RCIM_Cubic>,<at 0.4682 0.0526>,<at 0.7505 0.024>,<at 1 0.0104>]}
       }}))
    ScaleA=true
    ScaleRGB=true
   )
   &ScaleSpriteSize(Uniform Curve Sprite Scale=DataInterface() {propertyValues {Curve {keys [
        {arriveTangent=4.1785, interpMode=RCIM_Cubic, leaveTangent=4.1785, tangentMode=RCTM_User, time=0, value=0}
        {arriveTangent=0.9995, interpMode=RCIM_Cubic, leaveTangent=0.9995, tangentMode=RCTM_User, time=0.2223, value=0.6825}
        {arriveTangent=-0.0405, interpMode=RCIM_Cubic, leaveTangent=-0.0405, tangentMode=RCTM_User, time=1, value=1}
       ]}}})
   /SetVariables(
    Particles.FireColor:LinearColor=&MakeLinearColorFromVectorAndFloat(
     "Vector (RGB)"=&MakeVectorFromLinearColorRGB(Color=&ColorFromCurve(ColorCurve=DataInterface() {propertyValues {
         AlphaCurve {keys [<at 0 0.6 RCIM_Cubic>,<at 1 0 RCIM_Cubic>]}
         BlueCurve {keys [
           <at 0 0.0288>
           {arriveTangent=0.0134, interpMode=RCIM_Cubic, leaveTangent=0.0134, time=0.1567, value=0.2442}
           <at 0.2997 0.2281>
           <at 0.4884 0.0353>
           <at 0.6314 0.0169>
           <at 0.877 0.0089 RCIM_Cubic>
          ]}
         GreenCurve {keys [
           <at 0 0.0721>
           {arriveTangent=0.0353, interpMode=RCIM_Cubic, leaveTangent=0.0353, time=0.1567, value=0.6426}
           <at 0.2997 0.5979>
           <at 0.4884 0.0894>
           <at 0.6314 0.041>
           <at 0.877 0.0196 RCIM_Cubic>
          ]}
         RedCurve {keys [
           <at 0 0.1496>
           {arriveTangent=0.0656, interpMode=RCIM_Cubic, leaveTangent=0.0656, time=0.1567, value=1.2}
           <at 0.2997 1.1263>
           <at 0.4884 0.1817>
           <at 0.6314 0.0917>
           <at 0.877 0.0521 RCIM_Cubic>
          ]}
        }}))
     "Float (Alpha)"=&FloatFromCurve(FloatCurve=DataInterface() {propertyValues {Curve {
         defaultValue=0
         keys [
          {arriveTangent=-3.2548, interpMode=RCIM_Cubic, leaveTangent=-3.2548, tangentMode=RCTM_User, time=0, value=1}
          {interpMode=RCIM_Cubic, tangentMode=RCTM_User, time=0.35, value=0}
         ]
        }}})
    )
    Particles.FireSize:Vector2f=&Multiply_Vector2D_ByFloat(
     Float=&MergeHelper_FloatSwitch(0=50, 1=50, 2=70, 3=?, Index=$Particles.SpawnGroup, Num=::NewEnumerator1)
     Vector2D=&Vector2DFromFloat(Value=&FloatFromCurve(
       FloatCurve=DataInterface() {propertyValues {Curve {
          defaultValue=0
          keys [
           {arriveTangent=0.1677, interpMode=RCIM_Cubic, leaveTangent=0.1677, tangentMode=RCTM_User, time=0, value=0.2}
           {interpMode=RCIM_Cubic, tangentMode=RCTM_User, time=1, value=1}
          ]
         }}}
       Scale Curve=&Multiply_Float(
        A=&RandomRangeFloat(Minimum=0.8, Maximum=1)
        B=&FN_FrontEndFloatModifier(Gameplay=1, Front End=0.7)
       )
      ))
    )
   ) {isSetParameters=true, module=SetVariables_#4}
   /Game/Effects/Niagara/Emitters/Water/FN_DistanceKillParticles()
   /SetVariables(Particles.Smoke SubImageIndex:NiagaraFloat=&FloatFromCurve(FloatCurve=DataInterface() {propertyValues {Curve {defaultValue=1, keys [{tangentMode=RCTM_User, time=0, value=0}, {tangentMode=RCTM_User, time=1, value=63}]}}})) {isSetParameters=true}
  ]
  rendererClasses [&NiagaraSpriteRendererProperties]
  renderers [
   {rendererClass=&NiagaraSpriteRendererProperties}
   {rendererClass=&NiagaraSpriteRendererProperties, rendererIndex=1}
  ]
 }
 Flames_Decal {
  emitterUpdateScript [
   &EmitterState()
   &SpawnRate(SpawnRate=4, Use Spawn Probability=false)
  ]
  particleSpawnScript [
   &PartitionParticles(Number Of Partitions=4)
   &InitializeParticle(
    Color=<rgba 15 1.9685 0 1>
    Lifetime Max=1.8
    Lifetime Min=1.3
    Lifetime Mode=::NewEnumerator1
    Sprite Size=<65.5018 91.5368>
    Sprite Size Mode=::NewEnumerator4
    UsePositionOffset=false
    Write Material Random=true
   )
  ]
  particleUpdateScript [
   &ParticleState()
   &SkeletalMeshLocation(
    Bone / Socket Index=&SelectIntFromArray(Direct Array Index=$Particles.PartitionParticles.Partition, Int Selection Array=DataInterface() {propertyValues {IntData [10,11,12,15]}})
    Bone Sampling Mode=::NewEnumerator4
    MaskSpawnGroup=false
    Sampled Position Offset=&MergeHelper_VectorSwitch(0=<0 0 15>, 1=<0 0 -15>, 2=<0 0 20>, 3=<0 0 18.2799>, Index=$Particles.PartitionParticles.Partition, Num=::NewEnumerator2)
    Skeletal Mesh=$System.Skeletal Mesh Reader
   )
   &DynamicMaterialParameters(
    Index 0 Param 1=1.7
    Index 0 Param 2=1
    Index 0 Param 3=1
    Index 0 Param 4=&RandomRangeFloat(Minimum=0, Maximum=1)
    Module.Index0Float0Write=true
    Module.Index0Float1Write=true
    Module.Index0Float2Write=true
    Module.Index0Float3Write=true
   )
   &ScaleColor(
    Scale Alpha=&FloatFromCurve(
     FloatCurve=DataInterface() {propertyValues {Curve {keys [
         <at 0 0>
         <at 0 0>
         {arriveTangent=2.2236, interpMode=RCIM_Cubic, leaveTangent=2.2236, tangentMode=RCTM_User, time=0.2, value=0.65}
         {interpMode=RCIM_Cubic, tangentMode=RCTM_User, time=0.5, value=1}
         {arriveTangent=-2.604, interpMode=RCIM_Cubic, leaveTangent=-2.604, tangentMode=RCTM_User, time=0.8, value=0.65}
         <at 1 0>
        ]}}}
     Scale Curve=&RandomRangeFloat(Minimum=0.5, Maximum=0.75)
    )
    ScaleA=true
    ScaleRGB=true
   )
   &SolveForcesAndVelocity(Clamp Velocity=false, Limit Acceleration=false)
  ]
  rendererClasses [&NiagaraSpriteRendererProperties]
  renderers [{rendererClass=&NiagaraSpriteRendererProperties}]
 }
 Smoke_Back {
  emitterUpdateScript [
   &EmitterState()
   &SpawnRate(
    SpawnRate=&Multiply_Float(
     A=20
     B=&FN_FrontEndFloatModifier(Gameplay=&ClampFloat(Float=$User.ForwardDot, Min=0.3, Max=0.5), Front End=1)
    )
    Use Spawn Probability=false
   )
  ]
  particleSpawnScript [
   &InitializeParticle(
    Color=<rgba 0.0729 0.0239 0.011 0.5>
    Lifetime Max=1.2
    Lifetime Min=0.7
    Lifetime Mode=::NewEnumerator1
    Sprite Rotation Mode=::NewEnumerator2
    Sprite Size Mode=::NewEnumerator3
    Uniform Sprite Size=&Multiply_Float(
     A=&RandomRangeFloat(Minimum=90, Maximum=100)
     B=&FN_FrontEndFloatModifier(Gameplay=1, Front End=0.8)
    )
    UsePositionOffset=false
    Write Material Random=true
   )
   &SkeletalMeshLocation(
    Bone / Socket Index=12
    Bone Sampling Mode=::NewEnumerator4
    Interpolate Mesh Sampling=false
    MaskSpawnGroup=false
    Orientation Sampling=::NewEnumerator1
    Sampled Position Offset=&AddVectorToPosition(Position=<-20 0 20>, Vector=&RandomRangeVector(Minimum=<0 0 0>, Maximum=<-50 0 10>))
    Skeletal Mesh=$System.Skeletal Mesh Reader
   )
   &AddVelocity(Velocity=&Multiply_VectorByFloat(
     Vector=&RandomRangeVector(Minimum=<0 0 0>, Maximum=<0 -75 0>)
     Float=&FN_FrontEndFloatModifier(Gameplay=0, Front End=1)
    ))
   &CameraOffset(Camera Offset Amount=10)
   &DynamicMaterialParameters(
    Index 0 Param 1=1
    Index 0 Param 2=1
    Index 0 Param 3=1
    Index 0 Param 4=1
    Module.Param0WriteEnabled=true
    Module.Param1WriteEnabled=true
    Module.Param2WriteEnabled=true
    Module.Param3WriteEnabled=true
   )
  ]
  particleUpdateScript [
   &InheritSourceMovement(Applied Position Scale=<0.9 0.9 1>, ApplyForce=false, ApplyPosition=true, ApplyPositionFalloff=false)
   &SolveForcesAndVelocity(Clamp Velocity=false, Limit Acceleration=false)
   &ParticleState()
   &ScaleColor(
    Scale Alpha=&Multiply_Float(
     A=&FloatFromCurve(
      FloatCurve=DataInterface() {propertyValues {Curve {keys [
          {interpMode=RCIM_Cubic, tangentMode=RCTM_User, time=0, value=0}
          {interpMode=RCIM_Cubic, leaveTangent=-2.6773, tangentMode=RCTM_Break, time=0.3, value=1}
          {tangentMode=RCTM_User, time=1, value=0}
         ]}}}
      Scale Curve=&Multiply_Float(
       A=&RandomRangeFloat(Minimum=0.5, Maximum=1)
       B=&FN_FrontEndFloatModifier(Gameplay=1, Front End=0.2)
      )
     )
     B=&FNTimeOfDayScalar(Night Scale=1, Morning Scale=0.2, Evening Scale=0.2)
    )
    Scale RGB=&MakeVectorFromLinearColorRGB(Color=&ColorFromCurve(ColorCurve=DataInterface() {propertyValues {
        AlphaCurve {keys [<at 0 1 RCIM_Cubic>,<at 1 0 RCIM_Cubic>]}
        BlueCurve {keys [<at 0 1 RCIM_Cubic>,<at 0.4682 0.0526>,<at 0.7505 0.024>,<at 1 0.0104>]}
        GreenCurve {keys [<at 0 1 RCIM_Cubic>,<at 0.4682 0.0526>,<at 0.7505 0.024>,<at 1 0.0104>]}
        RedCurve {keys [<at 0 1 RCIM_Cubic>,<at 0.4682 0.0526>,<at 0.7505 0.024>,<at 1 0.0104>]}
       }}))
    ScaleA=true
    ScaleRGB=true
   )
   &SubUVAnimation()
   &ScaleSpriteSize(Uniform Curve Sprite Scale=DataInterface() {propertyValues {Curve {keys [
        {arriveTangent=4.1785, interpMode=RCIM_Cubic, leaveTangent=4.1785, tangentMode=RCTM_User, time=0, value=0}
        {arriveTangent=0.9995, interpMode=RCIM_Cubic, leaveTangent=0.9995, tangentMode=RCTM_User, time=0.2223, value=0.6825}
        {arriveTangent=-0.0405, interpMode=RCIM_Cubic, leaveTangent=-0.0405, tangentMode=RCTM_User, time=1, value=1}
       ]}}})
  ]
  rendererClasses [&NiagaraSpriteRendererProperties]
  renderers [{rendererClass=&NiagaraSpriteRendererProperties}]
 }
}
summary {
 systemName=NS_Glider_DevilRock_Idle
 emitters [
  {emitterName=Ash_Wings, rendererClasses [&NiagaraSpriteRendererProperties]}
  {emitterName=Dark_Smoke_Wings, rendererClasses [&NiagaraSpriteRendererProperties]}
  {emitterName=Flames_Decal, rendererClasses [&NiagaraSpriteRendererProperties]}
  {emitterName=CPU_particles, rendererClasses [&NiagaraSpriteRendererProperties]}
  {emitterName=Fire, rendererClasses [&NiagaraSpriteRendererProperties]}
  {emitterName=Fire_and_Smoke, rendererClasses [&NiagaraSpriteRendererProperties]}
  {emitterName=Smoke_Back, rendererClasses [&NiagaraSpriteRendererProperties]}
 ]
 userVariables [
  User.bIsFrontEnd {type=NiagaraBool}
  User.bIsFrontEndPreview {type=NiagaraBool}
  User.bIsFullyDeployed {type=NiagaraBool}
  User.ForwardDot {type=NiagaraFloat, defaultValue=0, description=""}
  User.IsSkydiving {type=NiagaraBool}
  User.RibbonIDOffset {type=NiagaraInt32}
  User.RightDot {type=NiagaraFloat, defaultValue=0, description=""}
 ]
}
